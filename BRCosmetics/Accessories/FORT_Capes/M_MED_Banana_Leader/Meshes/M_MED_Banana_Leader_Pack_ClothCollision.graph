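# Cloth collision physics asset for the M_MED_Banana_Leader pack, a back accessory cosmetic in the capes folder.
# A single backpack_root body holds two overlapping, similarly oriented capsules of slightly different size, offset from the bone to approximate a collision volume.
# It has no constraints, so it defines only collision shapes with no jointed ragdoll setup.
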PhysicsAsset /BRCosmetics/Accessories/FORT_Capes/M_MED_Banana_Leader/Meshes/M_MED_Banana_Leader_Pack_ClothCollision
bodies {backpack_root [
  {
   shapeType=Capsule
   length=13.5
   radius=10.5
   shapeName=backpack_root_capsule
   center=<-0.0001 -43.3419 -11.1791>
   rotation=<rot -90 90 -90.0002>
  }
  {
   shapeType=Capsule
   length=15
   radius=12.25
   shapeName=backpack_root_capsule
   center=<-0.0001 -40.1168 2.4417>
   rotation=<rot -90 90 -90.0001>
  }
 ]}
constraints []
